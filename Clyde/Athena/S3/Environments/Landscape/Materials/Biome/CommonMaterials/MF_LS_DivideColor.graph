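# MF_LS_DivideColor is a material function for landscape biome materials that re-tints a base color toward a biome color.
# It divides Basecolor by BiomeColorDiv and multiplies the result by BiomeColor, swapping the reference tint for the biome's.
# BiomeColorMask blends between the original and re-tinted colors.
# The UseBiomeColorMask static switch selects that blend or passes Basecolor through, and the output is saturated.
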MaterialFunction /Clyde/Athena/S3/Environments/Landscape/Materials/Biome/CommonMaterials/MF_LS_DivideColor
graph {nodes [
  n1 = FunctionInput() {inputName=BiomeColorDiv, sortPriority=2}
  n2 = FunctionInput() {inputName=Basecolor, sortPriority=4}
  n3 = LinearInterpolate(
   A=@n2
   B=Multiply(A=FunctionInput() {inputName=BiomeColor, sortPriority=3}, B=Divide(A=@n2, B=@n1))
   Alpha=FunctionInput(Preview=Constant() {r=1}) {bUsePreviewValueAsDefault=true, inputName=BiomeColorMask, inputType=FunctionInput_Scalar, sortPriority=1}
  )
  n4 = FunctionInput(Preview=StaticBool()) {bUsePreviewValueAsDefault=true, inputName=UseBiomeColorMask, inputType=FunctionInput_StaticBool, previewValue=<64 64 64 6>}
  FunctionOutput(None=Saturate(None=StaticSwitch(True=@n3, False=@n2, Value=@n4) {desc=Z projection}))
 ]}
