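# Mobile-only landscape material function that outputs material attributes for the common gravel road surface.
# Samples a dirt texture by world-aligned Z projection, tints it, and multiplies it by its alpha with a floor to limit darkening.
# Chained MF_LS_DivideColor calls recolor the result using Mountain, Tropical and Desert masks from MF_ApolloTerrainMasks.
# Specular, roughness and a flat normal are constants; a DirtColor blend branch is left unconnected.
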MaterialFunction /Clyde/Athena/S3/Environments/Landscape/Materials/MaterialFunctions/MF_Mobile_Common_Gravel_Roads
graph {nodes [
  n1 = MaterialFunctionCall!(UseLandscapeCoords=FunctionInput(Preview=StaticBool()) {bUsePreviewValueAsDefault=true, inputName=UseLandscapeCoords, inputType=FunctionInput_StaticBool}) {materialFunction=/Game/Athena/Apollo/Environments/Landscape/Materials/Biome/CommonMaterials/MF_ApolloTerrainMasks}
  n2 = StaticBool() {value=true}
  n3 = Constant3Vector() {constant=<rgba 0.8 0.5869 0.4049 0>}
  n4 = Constant3Vector() {constant=<rgba 0.3465 0.3481 0.35 1>}
  n5 = MaterialFunctionCall!(
   Diffuse Texture Object=TextureObject() {texture=/Game/Environments/Apollo/Sets/Terrain/Textures/T_Apollo_Dirt_D_01}
   Spec Texture Object
   TextureSize=Constant() {r=550}
   --------------------------
   Use seperate settings for Z projection
   Use seperate texture for Z projection
   Texture Size Z projection
   Z Plane Diffuse Texture Object
   Z Plane Spec Texture Object
   "--------------------------      "
   ProjectionTransitionContrast
   Use Texture Alpha Heightmap to Refine the Transistion=StaticBool()
   HeightLerpTransitionContrast
   PickTheDiffuseHeightChannel
   TheDiffuseHeightChannel
   "--------------------------       "
   NormalTextureObject
   World Space Normals=StaticBool()
   "--------------------------  "
   Export Float 4 Diffuse=StaticBool() {value=true}
   Export Float 4 Spec
   "--------------------------     "
   WorldPosition=WorldPosition() {worldPositionShaderOffset=WPT_ExcludeAllShaderOffsets}.XYZ
   World Space Normal
  ) {materialFunction=/Engine/Functions/Engine_MaterialFunctions02/Texturing/WorldAlignedTextures_Complex}
  n6 = Multiply(
   A=VectorParameter() {group=Dirt, parameterName=DirtColor1, defaultValue=<rgba 0.95 1 1 1>}.RGB
   B=ComponentMask(None=@n5.Z Diffuse Texture) {b=true, g=true, r=true}
  )
  n7 = ComponentMask(None=@n5.Z Diffuse Texture) {a=true}
  n8 = MaterialFunctionCall!(
   UseBiomeColorMask=StaticBool() {value=true}
   BiomeColorMask=@n1.Mountain
   BiomeColorDiv=Constant3Vector() {constant=<rgba 1 0.7337 0.5061 0>}
   BiomeColor=Constant3Vector() {constant=<rgba 0.4275 0.4425 0.45 1>}
   Basecolor=Multiply(A=@n6, B=Max(A=@n7) {constB=0.85})
  ) {materialFunction=/Clyde/Athena/S3/Environments/Landscape/Materials/Biome/CommonMaterials/MF_LS_DivideColor}
  n9 = SetMaterialAttributes!(
   MaterialAttributes
   BaseColor=MaterialFunctionCall(
    UseBiomeColorMask=StaticBool() {value=true}
    BiomeColorMask=@n1.Desert
    BiomeColorDiv=Constant3Vector() {constant=<rgba 0.8 0.7395 0.64 0>}
    BiomeColor=Constant3Vector() {constant=<rgba 0.9 0.7425 0.27 1>}
    Basecolor=MaterialFunctionCall(UseBiomeColorMask=@n2, BiomeColorMask=@n1.Tropical, BiomeColorDiv=@n3, BiomeColor=@n4, Basecolor=@n8.Result) {materialFunction=&MF_LS_DivideColor}.Result
   ) {materialFunction=&MF_LS_DivideColor}.Result
   Specular=Constant() {r=0.3}
   Roughness=Constant() {r=0.9}
   Normal=Constant3Vector() {constant=<rgba 0 0 1 0>}
  ) {attributeSetTypes [
    #1
    #2
    #3
    #4
   ]}
  FunctionOutput(None=@n9) {desc=Mobile Landscape Only, outputName=Material, sortPriority=-1}
  LinearInterpolate(
   A=Multiply(A=VectorParameter() {group=Dirt, parameterName=DirtColor, defaultValue=<rgba 0.45 0.3773 0.225 1>}.RGB, B=@n6)
   B=@n6
   Alpha=Constant() {r=0.5}
  )
 ]}
